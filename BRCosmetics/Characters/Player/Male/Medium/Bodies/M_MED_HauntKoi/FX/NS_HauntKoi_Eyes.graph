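# Niagara eye effect for the HauntKoi character cosmetic, with red glows and lights on the skeletal mesh.
# Glow sprites and two lights follow User.SkeletalMesh, using partition index to pick bone/socket and offset per eye.
# Loop-age curves flicker the lights, while a separate unattached mesh emitter fades in and out, shifting blue to red.
# User.bisHeadHidden switches the lights off and sets VisibilityTag on glow and mesh particles; User.IsSkydiving is unused.
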NiagaraSystem /BRCosmetics/Characters/Player/Male/Medium/Bodies/M_MED_HauntKoi/FX/NS_HauntKoi_Eyes
dependencies {
 usedDataInterfaces [&NiagaraDataInterfaceCurve, &NiagaraDataInterfaceVectorCurve]
 usedDynamicInputs []
 usedModules [&SystemState]
}
emitters {
 Glow {
  emitterUpdateScript [
   &EmitterState()
   &SpawnRate(SpawnRate=24, Use Spawn Probability=false)
  ]
  particleSpawnScript [
   /Niagara/Modules/Spawn/Initialization/InitializeParticle(
    Color=<rgba 3 0 0 1>
    Lifetime=&UniformRangedFloat(Minimum=0.45, Maximum=0.5)
    Module.Write Color=true
    Module.Write Lifetime=true
    Module.Write Mass=true
    Module.Write Material Random=true
    Module.Write Position=true
    Module.Write Scale=false
    Module.Write SpriteRotation=false
    Module.Write SpriteSize=true
    Position=&SimulationPosition()
    Sprite Size=&Vector2DFromFloat(Value=&UniformRangedFloat(Minimum=9, Maximum=9))
   )
   &DynamicMaterialParameters(
    Index 0 Param 1=&RandomRangeFloat(Minimum=0, Maximum=1)
    Module.Param0WriteEnabled=true
    Module.Param1WriteEnabled=false
    Module.Param2WriteEnabled=false
    Module.Param3WriteEnabled=false
   )
   &PartitionParticles()
   &FN_SetSpawnGroup(Number of Spawn Groups=2) {enabled=false}
   &CameraOffset(Camera Offset Amount=2)
  ]
  particleUpdateScript [
   &ParticleState()
   &SkeletalMeshLocation(
    Bone / Socket Index=$Particles.PartitionParticles.Partition
    Inherit Velocity=false
    Interpolate Mesh Sampling=false
    MaskSpawnGroup=false
    Mesh Sampling Type=::NewEnumerator5
    Sampled Position Offset=&Lerp_Vector(A=<0 1.5 -1.4>, B=<0 1.5 1.4>, Alpha=&MakeFloatFromInt(Int=$Particles.PartitionParticles.Partition))
    Sampled Position Offset Coordinate Space:ENiagaraMeshTransforms=::NewEnumerator2
    Skeletal Mesh=$User.SkeletalMesh
    Socket Sampling Mode=::NewEnumerator5
   )
   &ScaleColor(
    Scale Alpha=&FloatFromCurve(FloatCurve=DataInterface() {propertyValues {Curve {keys [<at 0 0>,<at 0.25 1>,<at 0.8107 1>,<at 1 0>]}}})
    ScaleA=true
    ScaleRGB=true
   )
   &ScaleSpriteSize(Scale Factor=&Vector2DFromFloat(Value=&FloatFromCurve(FloatCurve=DataInterface() {propertyValues {Curve {keys [
          {arriveTangent=-0.0036, interpMode=RCIM_Cubic, leaveTangent=-0.0036, tangentMode=RCTM_User, time=0, value=1}
          {arriveTangent=-2.2201, interpMode=RCIM_Cubic, leaveTangent=-2.2201, tangentMode=RCTM_User, time=1, value=0.1}
         ]}}})))
   /SetVariables(Particles.VisibilityTag=&MakeCustomIntFromBool(Bool=$User.bisHeadHidden, True INT=1, False INT=0)) {isSetParameters=true}
  ]
  rendererClasses [&NiagaraSpriteRendererProperties]
  renderers [{rendererClass=&NiagaraSpriteRendererProperties}]
 }
 Light {
  emitterUpdateScript [
   &EmitterState(Inactive Response=::NewEnumerator1, Life Cycle Mode=::NewEnumerator1, Loop Duration=2, UseLoopDelay=true)
   &SpawnBurst_Instantaneous(Spawn Count=2, Use Spawn Probability=false)
  ]
  particleSpawnScript [
   &InitializeParticle(
    Color=<rgba 1 0 0 4>
    Lifetime=2
    Module.Write Color=true
    Module.Write Lifetime=true
    Module.Write Mass=false
    Module.Write Material Random=true
    Module.Write Position=false
    Module.Write Scale=false
    Module.Write SpriteRotation=false
    Module.Write SpriteSize=true
    Position=&SimulationPosition()
    Sprite Size=<75 75>
   )
   &DynamicMaterialParameters(
    Index 0 Param 1=&UniformRangedFloat(Minimum=0.7, Maximum=0.9)
    Index 0 Param 2=&UniformRangedFloat(Minimum=0.05, Maximum=0.15)
    Index 0 Param 3=1
    Module.Param0WriteEnabled=true
    Module.Param1WriteEnabled=true
    Module.Param2WriteEnabled=false
    Module.Param3WriteEnabled=false
   ) {module=DynamicMaterialParameters002}
   &PartitionParticles()
   &FN_SetSpawnGroup(Number of Spawn Groups=2) {enabled=false}
  ]
  particleUpdateScript [
   &ParticleState()
   &SolveForcesAndVelocity(Clamp Velocity=false, Limit Acceleration=false)
   &SkeletalMeshLocation(
    Bone / Socket Index=$Particles.PartitionParticles.Partition
    Inherit Velocity=false
    Interpolate Mesh Sampling=false
    MaskSpawnGroup=false
    Mesh Sampling Type=::NewEnumerator5
    Sampled Position Offset=&Lerp_Vector(A=<0 2 -1.5>, B=<0 2 1.5>, Alpha=&MakeFloatFromInt(Int=$Particles.PartitionParticles.Partition))
    Sampled Position Offset Coordinate Space:ENiagaraMeshTransforms=::NewEnumerator2
    Skeletal Mesh=$User.SkeletalMesh
    Socket Sampling Mode=::NewEnumerator5
   )
   &ScaleColor(
    Scale RGB=&Multiply_Vector(
     A=&MakeVectorFromLinearColorRGB(Color=&LerpLinearColors(
       StartColor=<rgba 1 0.2078 0.0627 1>
       EndColor=<rgba 1 0.2352 0.1814 1>
       LerpFactor=&FloatFromCurve(
        FloatCurve=DataInterface() {propertyValues {Curve {keys [<at 0 1>,<at 0.1069 0.0895>,<at 0.3206 0.7423>,<at 0.4907 0.1444>,<at 0.6992 0.8948>,<at 0.9061 0.1261>,<at 1 1>]}}}
        CurveIndex=$Emitter.NormalizedLoopAge
        Scale Curve=0.9
       )
      ))
     B=&VectorFromFloat(Value=&FloatFromCurve(
       FloatCurve=DataInterface() {propertyValues {Curve {keys [<at 0 1>,<at 0.0584 1.1216>,<at 0.0797 0.6691>,<at 0.1464 0.526>,<at 0.222 1.0327>,<at 0.2705 0.6844>,<at 0.3166 0.8798>,<at 0.3641 0.5477>,<at 0.4407 0.6405>,<at 0.4815 1.1358>,<at 0.5388 1.0006>,<at 0.58 0.5314>,<at 0.6479 0.8407>,<at 0.6937 0.7459>,<at 0.737 1.0896>,<at 0.8125 0.5703>,<at 0.8909 0.8732>,<at 0.9449 0.6186>,<at 1 1>]}}}
       CurveIndex=$Emitter.NormalizedLoopAge
      ))
    )
    ScaleA=true
    ScaleRGB=true
   )
   &FN_Light_v2(
    Light Exponent=1
    Light Radius=5
    LightEnabled:NiagaraBool=&Boolean_NotOperation(Bool=$User.bisHeadHidden)
    Scale Color=3
   )
   &ScaleSpriteSize(Scale Factor=&Vector2DFromFloat(Value=&FloatFromCurve(
      FloatCurve=DataInterface() {propertyValues {Curve {keys [<at 0 1>,<at 0.0562 0.9129>,<at 0.0748 1.1321>,<at 0.1235 1.3668>,<at 0.2098 0.8951>,<at 0.2589 1.1544>,<at 0.2987 1.0516>,<at 0.3338 1.2874>,<at 0.3749 1.3981>,<at 0.4815 0.8928>,<at 0.5082 1.113>,<at 0.5446 1.2763>,<at 0.6237 1.1354>,<at 0.6838 1.0572>,<at 0.7109 0.8761>,<at 0.7888 0.7005>,<at 0.8531 1.2103>,<at 0.9284 1.1544>,<at 1 1>]}}}
      CurveIndex=$Emitter.NormalizedLoopAge
     )))
  ]
  rendererClasses [&NiagaraLightRendererProperties]
  renderers [{rendererClass=&NiagaraLightRendererProperties, rendererIndex=0}]
 }
 Particle_Emitter {
  emitterUpdateScript [
   &EmitterState(Life Cycle Mode=::NewEnumerator1, Loop Behavior=::NewEnumerator1, Loop Duration Mode=::NewEnumerator1, UseLoopDelay=false)
   &SpawnBurst_Instantaneous(Use Spawn Probability=false, UseLoopCountLimit=false)
   &SpawnRate(SpawnRate=2, Use Spawn Probability=false) {enabled=false}
  ]
  particleSpawnScript [
   /Niagara/Modules/Spawn/Initialization/V2/InitializeParticle(
    Lifetime=6
    Mesh Scale Mode=::NewEnumerator3
    Mesh Uniform Scale=1.05
    Position Offset=<4 0 0>
    UsePositionOffset=true
    Write Material Random=true
   )
   &Color(Color=&MakeLinearColorFromVectorAndFloat("Vector (RGB)"=<2 2 2>, "Float (Alpha)"=1))
   &DynamicMaterialParameters(Write Parameter Index 0=false) {enabled=false}
   &InitialMeshOrientation(
    Mesh Axis To Orient=<0 1 0>
    Mesh Orientation Mode=::NewEnumerator2
    Orientation Vector=<0 1 0>
    Rotation=<0 0.25 0>
    Use Rotation Vector=true
   )
  ]
  particleUpdateScript [
   &ParticleState(Kill Particles When Lifetime Has Elapsed=false, Let Infinitely Lived Particles Die When Emitter Deactivates=true)
   &ScaleColor(
    Scale Alpha=&FloatFromCurve(FloatCurve=DataInterface() {propertyValues {Curve {keys [<at 0 0 RCIM_Cubic>,<at 0.5 1 RCIM_Cubic>,<at 1 0 RCIM_Cubic>]}}})
    Scale RGB=&VectorFromCurve(VectorCurve=DataInterface() {propertyValues {
       XCurve {keys [
         <at 0 0.2276 RCIM_Cubic>
         {arriveTangent=8.2463, interpMode=RCIM_Cubic, leaveTangent=8.2463, time=0.2, value=0.7259}
         {arriveTangent=6.5927, interpMode=RCIM_Cubic, leaveTangent=6.5927, time=0.7, value=6}
         <at 1 6>
        ]}
       YCurve {keys [
         <at 0 2.7754 RCIM_Cubic>
         {arriveTangent=-3.9648, interpMode=RCIM_Cubic, leaveTangent=-3.9648, time=0.2, value=2.4246}
         {arriveTangent=-2.5971, interpMode=RCIM_Cubic, leaveTangent=-2.5971, time=0.7, value=0}
         <at 1 0.3469>
        ]}
       ZCurve {keys [
         <at 0 6 RCIM_Cubic>
         {arriveTangent=-7.8615, interpMode=RCIM_Cubic, leaveTangent=-7.8615, time=0.2, value=6}
         {arriveTangent=-6.3938, interpMode=RCIM_Cubic, leaveTangent=-6.3938, time=0.7, value=0.497}
         <at 1 0.885>
        ]}
      }})
    ScaleA=true
    ScaleRGB=true
   )
   &DynamicMaterialParameters(
    Index 0 Param 2=&FloatFromCurve(FloatCurve=DataInterface() {propertyValues {Curve {keys [<at 0.4 0 RCIM_Cubic>, {interpMode=RCIM_Cubic, leaveTangent=2.8354, time=1, value=1}]}}})
    Index 0 Param 3=&FloatFromCurve(FloatCurve=DataInterface() {propertyValues {Curve {keys [
         {arriveTangent=0.0343, interpMode=RCIM_Cubic, time=0, value=0.8}
         {arriveTangent=6.2, interpMode=RCIM_Cubic, leaveTangent=6.2, time=0.4, value=1}
         <at 1 7 RCIM_Cubic>
        ]}}})
    Module.Param0WriteEnabled=false
    Module.Param1WriteEnabled=true
    Module.Param2WriteEnabled=true
    Module.Param3WriteEnabled=false
   ) {enabled=false, module=DynamicMaterialParameters001}
   /SetVariables(Particles.VisibilityTag=&MakeCustomIntFromBool(Bool=$User.bisHeadHidden, True INT=1, False INT=0)) {isSetParameters=true}
  ]
  rendererClasses [&NiagaraMeshRendererProperties]
  renderers [
   {rendererClass=&NiagaraMeshRendererProperties, rendererIndex=0}
   {rendererClass=&NiagaraMeshRendererProperties, rendererIndex=1}
  ]
 }
}
summary {
 systemName=NS_HauntKoi_Eyes
 emitters [
  {emitterName=Light, rendererClasses [&NiagaraLightRendererProperties]}
  {emitterName=Glow, rendererClasses [&NiagaraSpriteRendererProperties]}
  {emitterName=Particle_Emitter, rendererClasses [&NiagaraMeshRendererProperties]}
 ]
 userVariables [
  User.bisHeadHidden {type=NiagaraBool}
  User.IsSkydiving {type=NiagaraBool}
  User.SkeletalMesh {
   type=NiagaraDataInterfaceSkeletalMesh
   defaultValue {
    struct=NiagaraExt_VariableValue_DataInterface
    value {dataInterface=&self:NiagaraDataInterfaceSkeletalMesh_0}
   }
  }
 ]
}
