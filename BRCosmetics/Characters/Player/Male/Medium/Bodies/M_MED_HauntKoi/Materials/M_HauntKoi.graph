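# Skeletal-mesh material for the HauntKoi character body, layering animated emissive effects over the shared PreFX, SkinFX and PostFX character functions.
# Pre-skinned position builds angle/height UVs for noise-broken eye glow, gated by bIsHead, and Fresnel-masked nebula.
# The nebula uses sphere-masked swirl, flow-map distortion and offset RGB samples for chromatic split, with time-of-day scaling on emissive.
# Optional sine WPO sways height-masked vertices, except on mobile.
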Material /BRCosmetics/Characters/Player/Male/Medium/Bodies/M_MED_HauntKoi/Materials/M_HauntKoi
graph {
 properties {
  bUseMaterialAttributes=true
  bUsedWithMeshDeformer=true
  bUsedWithMorphTargets=true
  bUsedWithSkeletalMesh=true
  pixelDepthOffsetMode=PDOM_Legacy
  refractionMethod=RM_IndexOfRefraction
  shadingModel=MSM_FromMaterialExpression
  subsurfaceProfile=&SS_HeroSkin_02
  usedShadingModels=Default Lit | Subsurface | Clear Coat | Subsurface Profile | Cloth | Eye
 }
 nodes [
  n1 = NamedRerouteUsage()
  ComponentMask(None=@n1) {r=true}
  Constant()
  n2 = MaterialFunctionCall() {materialFunction=&BaseFN_CharacterFunction_PreFX}
  n3 = MaterialFunctionCall(
   PreskinnedPosition=@n2.Pre skinned Local Position
   Gmap=@n2.Pre skinned Local Position
   MaterialAttributes=@n2.Pre skinned Local Position
   AO=@n2.Pre skinned Local Position
  ) {materialFunction=&CharacterSkinFX}
  n4 = GetMaterialAttributes(None=@n3.Result) {attributeGetTypes [#1, #2, #3]}
  n5 = NamedRerouteUsage()
  n6 = VectorParameter() {group=ViperZero, parameterName=Fresnel, channelNames {a=Overall, b=Power, g=Subtract, r=Exponent}, defaultValue=<rgba 2 0.2 5 0>}
  n7 = Fresnel(ExponentIn=@n6.Exponent) {exponent=2}
  n8 = Subtract(A=OneMinus(None=Saturate(None=@n7)), B=@n6.Subtract) {constB=0.2}
  n9 = Multiply(A=ComponentMask(None=@n5) {g=true}, B=@n8)
  n10 = ComponentMask(None=@n1) {b=true}
  n11 = OneMinus(None=@n10)
  n12 = ScalarParameter() {group=ViperZero, parameterName=InvertMask}
  n13 = Power(Base=Multiply(A=LinearInterpolate(A=@n11, B=@n10, Alpha=@n12)) {constA=1})
  n14 = VertexInterpolator(VS=PreSkinnedPosition())
  n15 = Constant3Vector() {constant=<rgba 1 1 0 0>}
  n16 = AppendVector(A=ComponentMask(None=@n14.PS) {g=true, r=true}, B=@n15.R)
  n17 = PreSkinnedLocalBounds()
  n18 = Subtract(A=@n17.Max, B=@n17.Max)
  n19 = MaterialFunctionCall(
   X=MaterialFunctionCall!(Normalized V1=Normalize(VectorInput=@n16), Normalized V2=Normalize(VectorInput=@n15)) {materialFunction=/Engine/Functions/Engine_MaterialFunctions02/Math/AngleBetweenVectors}.0-1
   Y=Divide(A=ComponentMask(None=@n14.PS) {b=true}, B=ComponentMask(None=@n18) {b=true})
  ) {materialFunction=&MakeFloat2}
  n20 = Subtract(
   A=@n19.Result
   B=ComponentMask(None=VectorParameter() {group=ViperZero, parameterName=CorePosition, defaultValue=<rgba 0.5 0.5 0 1>}.RGB) {g=true, r=true}
  )
  n21 = Multiply(A=@n20, B=ScalarParameter() {defaultValue=24, group=ViperZero, parameterName=EyeSmokeTiling})
  n22 = Multiply(
   A=TextureSample(UVs=Panner(Coordinate=@n21) {speedY=-0.25}) {texture=&T_NoisePack_01}.R
   B=ScalarParameter() {defaultValue=0.5, group=ViperZero, parameterName=EyeSmokeStrength}
  )
  n23 = Multiply(
   A=Clamp(None=Subtract(A=@n13, B=@n22))
   B=Add(A=NamedRerouteUsage(), B=VectorParameter() {group=ViperZero, parameterName=EyeGlowColor, defaultValue=<rgba 0 0.5687 1 1>}.RGB)
  )
  n24 = ScalarParameter() {defaultValue=5, group=ViperZero, parameterName=EyeGlowStrength}
  n25 = Multiply(
   A=Power(Base=Multiply(A=@n23, B=@n24), Exp=ScalarParameter() {defaultValue=2, group=ViperZero, parameterName=EyeGlowPower})
   B=ScalarParameter() {group=ViperZero, parameterName=bIsHead}
  )
  n26 = ComponentMask(None=NamedRerouteUsage()) {b=true}
  n27 = Multiply(A=@n6.Overall, B=Power(Base=Multiply(A=@n8, B=@n26)) {constExponent=3})
  n28 = ComponentMask(None=Constant3Vector() {constant=<rgba 0.5 0.5 0 0>}) {g=true, r=true}
  n29 = ScalarParameter() {defaultValue=0.75, group=ViperZero, parameterName=SwirlRadius}
  n30 = Constant()
  n31 = Rotator(
   Coordinate=@n20
   Time=Multiply(
    A=SphereMask(A=@n20, B=@n28, Radius=@n29, Hardness=@n30)
    B=ScalarParameter() {defaultValue=6, group=ViperZero, parameterName=CoreSwirlStrength}
   )
  )
  n32 = MaterialFunctionCall(
   X=ScalarParameter() {defaultValue=1, group=ViperZero, parameterName=NebulaTilingX}
   Y=ScalarParameter() {defaultValue=1, group=ViperZero, parameterName=NebulaTilingY}
  ) {materialFunction=&MakeFloat2}
  n33 = VectorParameter() {group=ViperZero, parameterName=NebulaPanSpeed, defaultValue=<rgba 0 -0.2 0 1>}
  n34 = Panner(
   Coordinate=Multiply(A=@n31, B=ScalarParameter() {defaultValue=0.5, group=ViperZero, parameterName=SmokeSwirlTiling})
   Speed=ComponentMask(None=VectorParameter() {group=ViperZero, parameterName=SwirlPanSpeedandDirection, defaultValue=<rgba 0 -0.5 0 1>}.RGB) {g=true, r=true}
  ) {speedY=-0.5}
  n35 = Subtract(A=Multiply(A=TextureSample(UVs=@n34) {samplerType=SAMPLERTYPE_LinearColor, texture=/Game/Global/Textures/Default/Flowmaps/T_SwirlFlow_02}.RGB) {constB=2})
  n36 = ScalarParameter() {defaultValue=0.1, group=ViperZero, parameterName=DistortionStrength}
  n37 = Add(
   A=Panner(Coordinate=Multiply(A=@n31, B=@n32.Result), Speed=ComponentMask(None=@n33.RGB) {g=true, r=true}) {speedY=-0.8}
   B=ComponentMask(None=Multiply(A=@n35, B=@n36)) {g=true, r=true}
  )
  n38 = TextureObject() {texture=&T_ShieldCaustics}
  n39 = TextureSample(UVs=@n37, Tex=@n38) {texture=/Game/Global/Textures/Default/Caustics/T_Turbulence_Seamless}
  n40 = ComponentMask(None=@n37) {r=true}
  n41 = ScalarParameter() {defaultValue=0.025, group=ViperZero, parameterName=Nebula Chromatic Green Offset}
  n42 = ComponentMask(None=@n37) {g=true}
  n43 = TextureSample(UVs=AppendVector(A=Add(A=@n40, B=@n41) {constB=0.025}, B=@n42), Tex=@n38) {texture=&T_Turbulence_Seamless}
  n44 = ScalarParameter() {defaultValue=0.05, group=ViperZero, parameterName=Nebula Chromatic Blue Offset}
  n45 = TextureSample(UVs=AppendVector(A=Add(A=@n40, B=@n44) {constB=0.05}, B=@n42), Tex=@n38) {texture=&T_Turbulence_Seamless}
  n46 = Multiply(
   A=Multiply(
    A=MaterialFunctionCall(X=@n39.R, Y=@n43.R, Z=@n45.R) {materialFunction=&MakeFloat3}.Result
    B=VectorParameter() {group=ViperZero, parameterName=NebulaColorShift, defaultValue=<rgba 0.5208 0.5208 0.5208 0>}.RGB
   )
   B=ScalarParameter() {defaultValue=1, group=ViperZero, parameterName=NebulaGlowStrength}
  )
  n47 = Multiply(A=OneMinus(None=ComponentMask(None=@n5) {r=true}), B=Add(A=Multiply(A=@n9, B=@n25), B=Multiply(A=@n27, B=@n46)))
  n48 = VectorParameter() {group=ViperZero, parameterName=TOD_BaseEmissiveScalers, channelNames {a=Night, b=Evening, g=Day, r=Morning}, defaultValue=<rgba 1 1 1 1>}
  n49 = MaterialFunctionCall(Value_Morning=@n48.Morning, Value_Day=@n48.Morning, Value_Evening=@n48.Morning, Value_Night=@n48.Morning) {materialFunction=&Scalar_Time_of_Day}
  n50 = NamedRerouteUsage()
  n51 = ComponentMask(None=PreSkinnedPosition()) {b=true}
  n52 = VectorParameter() {parameterName=WPO_Scalaers, channelNames {a=ShiftWPO, b=PostWPOMult, g=PreSinMult, r=WP Scale}, defaultValue=<rgba 0.5 0.4 0.5 0>}
  n53 = Saturate(None=Subtract(A=Subtract(A=@n51) {constB=0}, B=@n52.ShiftWPO) {constB=30})
  n54 = Time()
  n55 = Multiply(A=ComponentMask(None=WorldPosition().XYZ) {b=true}, B=@n52.WP Scale) {constB=0.5}
  n56 = Sine(None=Multiply(A=Add(A=@n54, B=@n55), B=@n52.PreSinMult) {constB=0.4})
  n57 = Constant3Vector() {constant=<rgba 1 1 0 0>}
  n58 = Multiply(A=Subtract(A=Multiply(A=@n56, B=@n57), B=Multiply(A=@n57) {constB=0.5}), B=@n52.PostWPOMult) {constB=0.5}
  n59 = ShadingPathSwitch(Default=Add(A=@n50, B=Multiply(A=@n53, B=@n58)), Mobile=@n50) {desc=Migrated from FeatureLevelSwitch}
  n60 = MaterialFunctionCall(
   AO=@n2.AO
   Distance Blend=@n2.AO
   Fresnel=@n2.AO
   Gmap=@n2.AO
   Pre Skinned Local Position=@n2.AO
   Material=SetMaterialAttributes!(
    MaterialAttributes=@n4.MaterialAttributes
    Base Color=@n4.MaterialAttributes
    Emissive Color=Multiply(A=@n47, B=@n49.Result)
    World Position Offset=StaticSwitchParameter(True=@n59, False=@n50) {parameterName=Use WPO}
   ) {attributeSetTypes [#1, #2, #3]}
  ) {materialFunction=&BaseFN_CharacterFunction_PostFX}
  MaterialFunctionCall(In=@n60.Result) {materialFunction=&CINE_MF_CharPostFX}
  n61 = CameraVectorWS()
  n62 = VectorParameter() {parameterName=Axis, defaultValue=<rgba 0 0 1 1>}
  n63 = MaterialFunctionCall() {materialFunction=&ObjectPivotPoint}
  n64 = ComponentMask(None=@n63.Object Pivot Location) {b=true}
  n65 = ScalarParameter() {defaultValue=96, group=ViperZero, parameterName=PivotZAdjustment}
  n66 = Divide(
   A=ScalarParameter() {defaultValue=200, group=ViperZero, parameterName=A_Rad}
   B=ScalarParameter() {defaultValue=97, group=Stars, parameterName=ActorRadius, sortPriority=0}
  )
  MaterialFunctionCall!(
   Plane Normal Axis=@n61
   Plane X Axis=Normalize(VectorInput=CrossProduct(A=@n61, B=@n62.RGB))
   Plane Center=AppendVector(A=ComponentMask(None=@n63.Object Pivot Location) {g=true, r=true}, B=Add(A=@n64, B=@n65) {constB=48})
   UV Scale=Multiply(A=ScalarParameter() {defaultValue=100, group=ViperZero, parameterName=UVRadialScaleZero}, B=@n66)
  ) {materialFunction=/Game/Athena/Environments/Nexus/Materials/VirtualPlaneCoordinatesAbs}
  Multiply(A=@n65, B=@n66)
  NamedRerouteDeclaration(None=@n3.SkinFX_Mask) {name=SkinFX, nodeColor=<rgba 1 0.9609 0 1>}
  NamedRerouteDeclaration(None=@n4.Emissive Color) {name=SourceEmissive, nodeColor=<rgba 0 1 0.5078 1>}
  NamedRerouteDeclaration(None=@n4.World Position Offset) {name=WPO Base, nodeColor=<rgba 0 1 0.5313 1>}
  NamedRerouteDeclaration(None=AppendVector(A=@n19.Result, B=@n15.G)) {name=TESTME, nodeColor=<rgba 0.2188 0 1 1>}
  Normalize(VectorInput=Subtract(
    A=MaterialFunctionCall() {materialFunction=&ObjectPivotPoint}.Object Pivot Location
    B=CameraPositionWS()
   ))
  ObjectRadius()
  Transform(None=Constant3Vector() {constant=<rgba 0 1 0 0>}) {transformSourceType=TRANSFORMSOURCE_World, transformType=TRANSFORM_Local}
 ]
}
