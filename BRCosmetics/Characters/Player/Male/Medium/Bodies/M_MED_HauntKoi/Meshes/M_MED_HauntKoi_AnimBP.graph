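# Animation Blueprint for the HauntKoi male medium body cosmetic, copying the pose from the attached parent mesh and adjusting the shoulder spikes.
# A deform Control Rig runs first, then constraints tie six spike bones to weighted clavicle and spine targets.
# A second Control Rig blends in while targeting, and additive rotations tilt the spikes while the parachute is open.
# EventGraph interpolates both states into float alphas (Is Targeting, Is Skydiving) for smooth blending.
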AnimBlueprint /BRCosmetics/Characters/Player/Male/Medium/Bodies/M_MED_HauntKoi/Meshes/M_MED_HauntKoi_AnimBP
target_skeleton=&/M_MED_HauntKoi_Skeleton
graphs [
 AnimGraph {nodes [
   n1 = Control Rig(Source=Copy Pose From Mesh() {bUseAttachedParent=true}) {
    alphaBoolBlend {blendOption=Linear}
    controlRigAssetReference {blueprintRigClass=&/M_MED_DeformRig_HauntKoi_ControlRig.M_MED_DeformRig_HauntKoi_ControlRig_C}
   }
   n2 = "Constraint - spike_01_l (PPP)"!(
    ComponentPose=Local To Component(LocalPose=@n1)
    Alpha=1
    ConstraintWeights_0=0.5946
    ConstraintWeights_1:float=0.3839
    ConstraintWeights_2:float=0.0215
   ) -> (Pose) {
    alphaBoolBlend {blendOption=Linear}
    boneToModify {boneName=spike_01_l}
    constraintSetup [{targetBone {boneName=clavicle_l}}, {targetBone {boneName=spine_05}}, {targetBone {boneName=spine_04}}]
    constraintWeights [1,1,1]
   }
   n3 = "Constraint - spike_02_l (PP)"!(ComponentPose=@n2, Alpha=1, ConstraintWeights_0=0.7153, ConstraintWeights_1:float=0.2847) -> (Pose) {
    alphaBoolBlend {blendOption=Linear}
    boneToModify {boneName=spike_02_l}
    constraintSetup [{targetBone {boneName=clavicle_l}}, {targetBone {boneName=spine_05}}]
    constraintWeights [1,1]
   }
   n4 = "Constraint - spike_03_l (PP)"!(ComponentPose=@n3, Alpha=1, ConstraintWeights_0=0.8355, ConstraintWeights_1:float=0.1645) -> (Pose) {
    alphaBoolBlend {blendOption=Linear}
    boneToModify {boneName=spike_03_l}
    constraintSetup [{targetBone {boneName=clavicle_l}}, {targetBone {boneName=spine_05}}]
    constraintWeights [1,1]
   }
   n5 = "Constraint - spike_01_r (PPP)"!(ComponentPose=@n4, Alpha=1, ConstraintWeights_0=0.5946, ConstraintWeights_1:float=0.3839, ConstraintWeights_2:float=0.0215) -> (Pose) {
    alphaBoolBlend {blendOption=Linear}
    boneToModify {boneName=spike_01_r}
    constraintSetup [{targetBone {boneName=clavicle_r}}, {targetBone {boneName=spine_05}}, {targetBone {boneName=spine_04}}]
    constraintWeights [1,1,1]
   }
   n6 = "Constraint - spike_02_r (PP)"!(ComponentPose=@n5, Alpha=1, ConstraintWeights_0=0.7153, ConstraintWeights_1:float=0.2847) -> (Pose) {
    alphaBoolBlend {blendOption=Linear}
    boneToModify {boneName=spike_02_r}
    constraintSetup [{targetBone {boneName=clavicle_r}}, {targetBone {boneName=spine_05}}]
    constraintWeights [1,1]
   }
   n7 = "Constraint - spike_03_r (PP)"!(ComponentPose=@n6, Alpha=1, ConstraintWeights_0=0.8355, ConstraintWeights_1:float=0.1645) -> (Pose) {
    alphaBoolBlend {blendOption=Linear}
    boneToModify {boneName=spike_03_r}
    constraintSetup [{targetBone {boneName=clavicle_r}}, {targetBone {boneName=spine_05}}]
    constraintWeights [1,1]
   }
   n8 = Control Rig(Source=Component To Local(ComponentPose=@n7), Alpha=Get Is Targeting() -> (Is Targeting:double)) {
    alphaBoolBlend {blendOption=Linear}
    controlRigAssetReference {blueprintRigClass=&/M_MED_HauntKoi_CtrlRig.M_MED_HauntKoi_CtrlRig_C}
   }
   n9 = Get Is Skydiving() -> (Is Skydiving:double)
   n10 = "Transform (Modify) Bone - Bone: spike_01_l"(ComponentPose=Local To Component(LocalPose=@n8), Alpha=@n9, Translation="0,0,0", Rotation="-20,0,-5", Scale="1,1,1") {
    rotationMode=BMM_Additive
    rotationSpace=BCS_BoneSpace
    alphaBoolBlend {blendOption=Linear}
    boneToModify {boneName=spike_01_l}
    rotation {pitch=-20, roll=-5}
   }
   n11 = "Transform (Modify) Bone - Bone: spike_02_l"(ComponentPose=@n10, Alpha=@n9, Translation="0,0,0", Rotation="-20,0,-5", Scale="1,1,1") {
    rotationMode=BMM_Additive
    rotationSpace=BCS_BoneSpace
    alphaBoolBlend {blendOption=Linear}
    boneToModify {boneName=spike_02_l}
    rotation {pitch=-20, roll=-5}
   }
   n12 = "Transform (Modify) Bone - Bone: spike_03_l"(ComponentPose=@n11, Alpha=@n9, Translation="0,0,0", Rotation="-20,0,-5", Scale="1,1,1") {
    rotationMode=BMM_Additive
    rotationSpace=BCS_BoneSpace
    alphaBoolBlend {blendOption=Linear}
    boneToModify {boneName=spike_03_l}
    rotation {pitch=-20, roll=-5}
   }
   n13 = "Transform (Modify) Bone - Bone: spike_01_r"(ComponentPose=@n12, Alpha=@n9, Translation="0,0,0", Rotation="-20,0,-5", Scale="1,1,1") {
    rotationMode=BMM_Additive
    rotationSpace=BCS_BoneSpace
    alphaBoolBlend {blendOption=Linear}
    boneToModify {boneName=spike_01_r}
    rotation {pitch=-20, roll=-5}
   }
   n14 = "Transform (Modify) Bone - Bone: spike_02_r"(ComponentPose=@n13, Alpha=@n9, Translation="0,0,0", Rotation="-20,0,-5", Scale="1,1,1") {
    rotationMode=BMM_Additive
    rotationSpace=BCS_BoneSpace
    alphaBoolBlend {blendOption=Linear}
    boneToModify {boneName=spike_02_r}
    rotation {pitch=-20, roll=-5}
   }
   n15 = "Transform (Modify) Bone - Bone: spike_03_r"(ComponentPose=@n14, Alpha=@n9, Translation="0,0,0", Rotation="-20,0,-5", Scale="1,1,1") {
    rotationMode=BMM_Additive
    rotationSpace=BCS_BoneSpace
    alphaBoolBlend {blendOption=Linear}
    boneToModify {boneName=spike_03_r}
    rotation {pitch=-20, roll=-5}
   }
   Output Pose(Result=Component To Local(ComponentPose=@n15)) {layerGroup=None}
  ]}
 EventGraph {nodes [
   n1 = Event BlueprintUpdateAnimation()
   n2 = Get bIsTargeting()
   > Set Is Targeting(Is Targeting:double=FInterpTo(Current=Get Is Targeting(), Target="To Float (Boolean)"(InBool=@n2), DeltaTime=@n1.DeltaTimeX, InterpSpeed=10)) -> (Output_Get:double)
   n3 = Get bIsParachuteOpened()
   > Set Is Skydiving(Is Skydiving:double=FInterpTo(Current=Get Is Skydiving(), Target="To Float (Boolean)"(InBool=@n3), DeltaTime=@n1.DeltaTimeX, InterpSpeed=10)) -> (Output_Get:double)
  ]}
]
variables [
 Is Skydiving {type=number}
 Is Targeting {type=number}
]
